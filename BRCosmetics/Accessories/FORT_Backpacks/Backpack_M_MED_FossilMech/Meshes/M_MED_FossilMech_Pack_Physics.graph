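# Physics asset for the FossilMech back bling, giving its lower jaw a physics body for secondary motion on the backpack.
# A capsule on the dyn_jaw_lower bone connects through a single constraint to a small capsule on the pack bone.
# The constraint locks twist and swing2, leaving swing1 as the only rotational axis, so the jaw moves like a hinge.
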PhysicsAsset /BRCosmetics/Accessories/FORT_Backpacks/Backpack_M_MED_FossilMech/Meshes/M_MED_FossilMech_Pack_Physics
bodies {
 dyn_jaw_lower [{shapeType=Capsule, length=10, radius=10, shapeName=dyn_jaw_lower_capsule, center=<10 0 0>, rotation=<rot 90 0 0>}]
 pack [{shapeType=Capsule, length=0.505, radius=0.505, shapeName=pack_capsule, rotation=<rot 0 -90 0>}]
}
constraints [{
  bone1Name=dyn_jaw_lower
  bone2Name=pack
  swing1LimitDegrees=10
  swing2LimitDegrees=0.01
  swing2Motion=Locked
  twistLimitDegrees=0.01
  twistMotion=Locked
 }]
