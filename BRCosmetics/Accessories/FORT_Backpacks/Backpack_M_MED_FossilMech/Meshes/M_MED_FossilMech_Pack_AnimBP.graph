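# Animation Blueprint for the FossilMech backpack, posing and physically animating its jaws and screens.
# One linear chain sets jaw bone rotations, adds RigidBody physics while jumping, springs the screens with Anim Dynamics (zero-limit copies in the front end), then offsets the pack.
# Two opposite additive bolt rotations sit at equal alpha.
# The update event toggles Open/Close on skydiving or tactical sprint, but the output chain reads neither, and several nodes are unconnected.
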AnimBlueprint /BRCosmetics/Accessories/FORT_Backpacks/Backpack_M_MED_FossilMech/Meshes/M_MED_FossilMech_Pack_AnimBP
target_skeleton=&/M_MED_FossilMech_Pack_Skeleton
graphs [
 AnimGraph {nodes [
   Blend Poses by bool(BlendTime_0=0.3, BlendTime_1=0.8) {blendType=Linear}
   Component To Local(ComponentPose=Local To Component())
   Component To Local()
   Component To Local()
   Local To Component(LocalPose=Use cached pose 'base'())
   Local To Component(LocalPose=Use cached pose 'base'())
   n1 = "Transform (Modify) Bone - Bone: dyn_jaw_base"(ComponentPose=Local To Component(LocalPose=Input Pose()), Alpha=1, Translation="0,0,0", Rotation="0,0,0", Scale="1,1,1") {
    rotationMode=BMM_Replace
    rotationSpace=BCS_BoneSpace
    scaleMode=BMM_Replace
    scaleSpace=BCS_BoneSpace
    translationMode=BMM_Replace
    translationSpace=BCS_BoneSpace
    alphaBoolBlend {blendOption=Linear}
    alphaScaleBiasClamp {bInterpResult=true, interpSpeedDecreasing=8, interpSpeedIncreasing=3}
    boneToModify {boneName=dyn_jaw_base}
   }
   n2 = "Transform (Modify) Bone - Bone: dyn_jaw_lower"(ComponentPose=@n1, Alpha=1, Translation="0,0,0", Rotation="0,0,0", Scale="1,1,1") {
    rotationMode=BMM_Replace
    rotationSpace=BCS_BoneSpace
    scaleMode=BMM_Replace
    scaleSpace=BCS_BoneSpace
    translationMode=BMM_Replace
    translationSpace=BCS_BoneSpace
    alphaBoolBlend {blendOption=Linear}
    alphaScaleBiasClamp {bInterpResult=true, interpSpeedDecreasing=8, interpSpeedIncreasing=3}
    boneToModify {boneName=dyn_jaw_lower}
   }
   n3 = "Transform (Modify) Bone - Bone: dyn_jaw_top"(ComponentPose=@n2, Alpha=1, Translation="0,0,0", Rotation="0,0,0", Scale="1,1,1") {
    rotationMode=BMM_Replace
    rotationSpace=BCS_BoneSpace
    scaleMode=BMM_Replace
    scaleSpace=BCS_BoneSpace
    translationMode=BMM_Replace
    translationSpace=BCS_BoneSpace
    alphaBoolBlend {blendOption=Linear}
    alphaScaleBiasClamp {bInterpResult=true, interpSpeedDecreasing=8, interpSpeedIncreasing=3}
    boneToModify {boneName=dyn_jaw_top}
   }
   n4 = Anim Dynamics - Bone: dyn_jaw_top(ComponentPose=@n3, Alpha=1, ExternalForce="0,0,0") {
    angularDampingOverride=0.99
    angularSpringConstant=2000
    bAngularSpring=true
    bGravityOverrideInSimSpace=true
    bOverrideAngularDamping=true
    bOverrideLinearDamping=true
    bUseGravityOverride=true
    gravityScale=0.35
    linearDampingOverride=0.99
    simulationSpace=BoneRelative
    alphaBoolBlend {blendOption=Linear}
    alphaScaleBiasClamp {bInterpResult=true, scale=0.2}
    boundBone {boneName=dyn_jaw_top}
    componentAppliedLinearAccClamp=<10000 10000 10000>
    componentLinearAccScale=<0.6 0.6 0.6>
    componentLinearVelScale=<0.6 0.6 0.6>
    gravityOverride {z=-500}
    physicsBodyDefinitions [{
      sphereCollisionRadius=0
      boundBone {boneName=dyn_jaw_top}
      boxExtents=<10 10 10>
      constraintSetup {angularLimitsMax=<0 0 5>, angularLimitsMin=<0 0 -5>, angularTarget=<1 0 0>}
      localJointOffset=<1 0 0>
     }]
   }
   n5 = "Transform (Modify) Bone - Bone: dyn_jaw_base"(ComponentPose=@n4, Alpha=1, Translation="0,0,0", Rotation="0,0,0", Scale="1,1,1") {
    rotationMode=BMM_Replace
    rotationSpace=BCS_BoneSpace
    scaleMode=BMM_Replace
    scaleSpace=BCS_BoneSpace
    translationMode=BMM_Replace
    translationSpace=BCS_BoneSpace
    alphaBoolBlend {blendOption=Linear}
    alphaScaleBiasClamp {bInterpResult=true, interpSpeedDecreasing=8, interpSpeedIncreasing=3}
    boneToModify {boneName=dyn_jaw_base}
   }
   n6 = "Transform (Modify) Bone - Bone: dyn_jaw_lower"(ComponentPose=@n5, Alpha=1, Translation="0,0,0", Rotation="0,-25,0", Scale="1,1,1") {
    rotationMode=BMM_Replace
    rotationSpace=BCS_BoneSpace
    scaleMode=BMM_Replace
    scaleSpace=BCS_BoneSpace
    translationMode=BMM_Replace
    translationSpace=BCS_BoneSpace
    alphaBoolBlend {blendOption=Linear}
    alphaScaleBiasClamp {bInterpResult=true, interpSpeedDecreasing=8, interpSpeedIncreasing=3}
    boneToModify {boneName=dyn_jaw_lower}
    rotation {yaw=-25}
   }
   n7 = "Transform (Modify) Bone - Bone: dyn_jaw_top"(ComponentPose=@n6, Alpha=1, Translation="0,0,0", Rotation="0,-120,0", Scale="1,1,1") {
    rotationMode=BMM_Replace
    rotationSpace=BCS_BoneSpace
    scaleMode=BMM_Replace
    scaleSpace=BCS_BoneSpace
    translationMode=BMM_Replace
    translationSpace=BCS_BoneSpace
    alphaBoolBlend {blendOption=Linear}
    alphaScaleBiasClamp {bInterpResult=true, interpSpeedDecreasing=8, interpSpeedIncreasing=3}
    boneToModify {boneName=dyn_jaw_top}
    rotation {yaw=-120}
   }
   n8 = RigidBody!(ComponentPose=@n7, Alpha="To Float (Boolean)"(InBool=Get bIsJumping()), OverrideWorldGravity:Vector="0,0,-980", ExternalForce) -> (Pose) {
    bOverrideWorldGravity=true
    overridePhysicsAsset=&/M_MED_FossilMech_Pack_Physics
    alphaBoolBlend {blendOption=Linear}
    alphaScaleBiasClamp {bInterpResult=true, interpSpeedDecreasing=5, interpSpeedIncreasing=6, scale=0.4}
    baseBoneRef {boneName=pack}
    componentAppliedLinearAccClamp=<1000 1000 1000>
    componentLinearAccScale=<0.2 0.2 0.6>
    componentLinearVelScale=<0.1 0.1 0.3>
    overrideWorldGravity {z=-980}
    simSpaceSettings {
     maxAngularAcceleration=20
     maxAngularVelocity=1
     maxLinearAcceleration=300
     maxLinearVelocity=150
     velocityScaleZ=0
     worldAlpha=1
     externalLinearDragV=<0.2 0.2>
    }
   }
   n9 = Anim Dynamics - Bone: dyn_jaw_top_screen(ComponentPose=@n8, Alpha=0.6, ExternalForce="0,0,0") {
    angularDampingOverride=0.99
    angularSpringConstant=2200
    bAngularSpring=true
    bGravityOverrideInSimSpace=true
    bOverrideAngularDamping=true
    bOverrideLinearDamping=true
    bUseGravityOverride=true
    gravityScale=0.35
    linearDampingOverride=0.99
    simulationSpace=BoneRelative
    alphaBoolBlend {blendOption=Linear}
    alphaScaleBiasClamp {scale=0.5}
    boundBone {boneName=dyn_jaw_top_screen}
    componentAppliedLinearAccClamp=<10000 10000 10000>
    componentLinearAccScale=<1 1 1>
    componentLinearVelScale=<1 1 1>
    gravityOverride {z=-500}
    physicsBodyDefinitions [{
      sphereCollisionRadius=0
      boundBone {boneName=dyn_jaw_top_screen}
      boxExtents=<10 10 10>
      constraintSetup {angularLimitsMax=<0 0 7>, angularLimitsMin=<0 0 -5>, angularTarget=<1 0 0>}
      localJointOffset=<1 0 0>
     }]
   }
   n10 = Anim Dynamics - Bone: dyn_jaw_lower_screen(ComponentPose=@n9, Alpha=0.8, ExternalForce="0,0,0") {
    angularDampingOverride=0.99
    angularSpringConstant=2000
    bAngularSpring=true
    bGravityOverrideInSimSpace=true
    bOverrideAngularDamping=true
    bOverrideLinearDamping=true
    bUseGravityOverride=true
    gravityScale=0.35
    linearDampingOverride=0.8
    simulationSpace=BoneRelative
    alphaBoolBlend {blendOption=Linear}
    alphaScaleBiasClamp {scale=0.5}
    boundBone {boneName=dyn_jaw_lower_screen}
    componentAppliedLinearAccClamp=<5000 5000 5000>
    componentLinearAccScale=<0.35 0.35 0.35>
    componentLinearVelScale=<0.25 0.25 0.25>
    gravityOverride {z=-500}
    physicsBodyDefinitions [{
      sphereCollisionRadius=0
      boundBone {boneName=dyn_jaw_lower_screen}
      boxExtents=<10 10 10>
      constraintSetup {angularLimitsMax=<0 0 0>, angularLimitsMin=<0 0 -30>, angularTarget=<1 0 0>}
      localJointOffset=<1 0 0>
     }]
   }
   n11 = Get bIsInFrontEnd()
   n12 = Anim Dynamics - Bone: dyn_jaw_top_screen!(ComponentPose=@n10, bAlphaBoolEnabled:bool=@n11, ExternalForce="0,0,0") -> (Pose) {
    alphaInputType=Bool
    angularDampingOverride=0.99
    angularSpringConstant=2200
    bAngularSpring=true
    bGravityOverrideInSimSpace=true
    bOverrideAngularDamping=true
    bOverrideLinearDamping=true
    bUseGravityOverride=true
    gravityScale=0.35
    linearDampingOverride=0.99
    simulationSpace=BoneRelative
    alphaBoolBlend {blendOption=Linear}
    alphaScaleBiasClamp {scale=0.5}
    boundBone {boneName=dyn_jaw_top_screen}
    componentAppliedLinearAccClamp=<10000 10000 10000>
    componentLinearAccScale=<1 1 1>
    componentLinearVelScale=<1 1 1>
    gravityOverride {z=-500}
    physicsBodyDefinitions [{
      sphereCollisionRadius=0
      boundBone {boneName=dyn_jaw_top_screen}
      boxExtents=<10 10 10>
      constraintSetup {angularLimitsMax=<0 0 0>, angularLimitsMin=<0 0 0>, angularTarget=<1 0 0>}
      localJointOffset=<1 0 0>
     }]
   }
   n13 = Anim Dynamics - Bone: dyn_jaw_lower_screen!(ComponentPose=@n12, bAlphaBoolEnabled:bool=@n11, ExternalForce="0,0,0") -> (Pose) {
    alphaInputType=Bool
    angularDampingOverride=0.99
    angularSpringConstant=2000
    bAngularSpring=true
    bGravityOverrideInSimSpace=true
    bOverrideAngularDamping=true
    bOverrideLinearDamping=true
    bUseGravityOverride=true
    gravityScale=0.35
    linearDampingOverride=0.8
    simulationSpace=BoneRelative
    alphaBoolBlend {blendOption=Linear}
    alphaScaleBiasClamp {scale=0.5}
    boundBone {boneName=dyn_jaw_lower_screen}
    componentAppliedLinearAccClamp=<5000 5000 5000>
    componentLinearAccScale=<0.35 0.35 0.35>
    componentLinearVelScale=<0.25 0.25 0.25>
    gravityOverride {z=-500}
    physicsBodyDefinitions [{
      sphereCollisionRadius=0
      boundBone {boneName=dyn_jaw_lower_screen}
      boxExtents=<10 10 10>
      constraintSetup {angularLimitsMax=<0 0 0>, angularLimitsMin=<0 0 0>, angularTarget=<1 0 0>}
      localJointOffset=<1 0 0>
     }]
   }
   n14 = "Transform (Modify) Bone - Bone: dyn_jaw_bolts"(ComponentPose=@n13, Alpha=1, Translation="0,0,0", Rotation="0,0,-3000", Scale="1,1,1") {
    rotationMode=BMM_Additive
    rotationSpace=BCS_BoneSpace
    alphaBoolBlend {blendOption=Linear}
    alphaScaleBiasClamp {bInterpResult=true, interpSpeedDecreasing=5, interpSpeedIncreasing=7}
    boneToModify {boneName=dyn_jaw_bolts}
    rotation {roll=-3000}
   }
   n15 = "Transform (Modify) Bone - Bone: dyn_jaw_bolts"(ComponentPose=@n14, Alpha=1, Translation="0,0,0", Rotation="0,0,3000", Scale="1,1,1") {
    rotationMode=BMM_Additive
    rotationSpace=BCS_BoneSpace
    alphaBoolBlend {blendOption=Linear}
    alphaScaleBiasClamp {bInterpResult=true, interpSpeedDecreasing=5, interpSpeedIncreasing=7}
    boneToModify {boneName=dyn_jaw_bolts}
    rotation {roll=3000}
   }
   n16 = "Transform (Modify) Bone - Bone: pack"(ComponentPose=@n15, Translation="0,2,-5", Rotation="0,0,-10") {
    rotationMode=BMM_Additive
    translationMode=BMM_Additive
    translationSpace=BCS_BoneSpace
    alphaBoolBlend {blendOption=Linear}
    boneToModify {boneName=pack}
    rotation {roll=-10}
    translation {y=2, z=-5}
   }
   Output Pose(Result=Component To Local(ComponentPose=@n16)) {layerGroup=None}
   Save cached pose 'base'() {cacheName=base}
   OR Boolean!(A=false, B=false, C:bool) -> (ReturnValue)
   Get Spinning_Rotator() -> (Spinning_Rotator:Rotator)
   Get bIsAccelerating2D()
   Get BoltRotate() -> (BoltRotate:double)
   Get bIsSprinting()
   Get bIsMoving2D()
   Get Spinning_Rotator_SPD2() -> (Spinning_Rotator_SPD2:Rotator)
  ]}
 EventGraph {nodes [
   Event BlueprintUpdateAnimation()
   > n1 = Branch(Condition=OR Boolean!(A=Get bIsSkyDiving(), B=Get bIsTacticalSprinting()) -> (ReturnValue))
   > Set Close(Close:bool=true) -> (Output_Get:bool)
   > Set Open(Open:bool=false) -> (Output_Get:bool)
   Set Open(execute=@n1.else, Open=true)
   > Set Close(Close=false)
   Branch(Condition=Get Close() -> (Close:bool))
   > Delay(Duration=0.2)
   > Set BoltRotate(BoltRotate:double=1) -> (Output_Get:double)
   n2 = float / float(A=GetRealTimeSeconds(), B=-720) {is=title K2Node_CallFunction}
   Set Spinning_Rotator(Spinning_Rotator:Rotator=MakeRotator(Roll=float * float(A=@n2, B=9000))) -> (Output_Get:Rotator)
   n3 = float / float(A=GetRealTimeSeconds(), B=-720)
   > Set Spinning_Rotator_SPD2(Spinning_Rotator_SPD2:Rotator=MakeRotator(Roll=float * float(A=@n3, B=50000))) -> (Output_Get:Rotator)
   TryGetPawnOwner()
   Sequence!(execute) -> (then_0, then_1, then_2:Exec, then_3:Exec, then_4:Exec)
   Branch(Condition=Get bIsInFrontEnd())
   Branch(Condition=NOT Boolean(A=Get bIsInFrontEnd()))
   Get bIsAccelerating2D()
   Get bIsSprinting()
   Get bIsMoving2D()
   Set Close(Close=OR Boolean!(A=Get bIsSprinting(), B=Get bIsTacticalSprinting(), C:bool=Get bIsSkyDiving()) -> (ReturnValue))
  ]}
]
variables [
 BoltRotate {type=number}
 Close {type=boolean}
 Open {type=boolean}
 Spinning_Rotator {type=Rotator}
 Spinning_Rotator_SPD2 {type=Rotator}
]
